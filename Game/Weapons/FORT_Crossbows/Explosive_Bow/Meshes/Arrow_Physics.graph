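# Physics asset for the arrow mesh of the explosive bow weapon, giving it simple collision for physics simulation.
# A single body on the arrow_nock bone holds one capsule, offset and rotated to lie along the length of the arrow shaft.
# The asset defines no constraints, so the arrow behaves as one rigid piece.
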PhysicsAsset /Game/Weapons/FORT_Crossbows/Explosive_Bow/Meshes/Arrow_Physics
bodies {arrow_nock [{shapeType=Capsule, length=82.5419, radius=3.2177, center=<42.2341 -0.1863 -0.124>, rotation=<rot 0 -90.0594 -89.8348>}]}
constraints []
